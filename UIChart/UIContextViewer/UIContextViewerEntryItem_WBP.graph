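# Widget Blueprint for one key/value row in the UI Context Viewer of the UIChart editor tooling, built on the C++ class UIContextViewerEntryItem.
# A horizontal box places the inherited key text inside a size box with a minimum width, then a small spacer and the inherited value text.
# Both labels share a grey bold Roboto style.
# The event graph holds only unconnected default events, so the Blueprint supplies layout alone.
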WidgetBlueprint /UIChart/UIContextViewer/UIContextViewerEntryItem_WBP
graphs [EventGraph {nodes [
   Event PreConstruct()
   Event Construct()
   Event Tick()
  ]}]
variables []
widgets {
 info {inheritedWidgetCount=2, parentClass=/Script/UIChartEditor.UIContextViewerEntryItem, rootWidgetClass=&HorizontalBox, widgetCount=5}
 widgets [
  HorizontalBox {widgetClassPath=&HorizontalBox, widget=HorizontalBox}
  SizeBox_1 {
   widgetClassPath=&SizeBox
   parent=HorizontalBox
   widget=SizeBox_1
   properties {bOverride_MinDesiredWidth=true, heightOverride=300, minDesiredWidth=300, widthOverride=200}
   slot {class=HorizontalBoxSlot, properties {padding=<ltrb 6 0 0 0>}}
  }
  KeyTextBLock {
   widgetClassPath=&TextBlock
   bInherited=true
   parent=SizeBox_1
   widget=KeyTextBLock
   properties {
    colorAndOpacity {colorUseRule=UseColor_Specified, specifiedColor=<rgba 0.5765 0.5765 0.5765 1>}
    font {
     bForceMonospaced=false
     bMaterialIsStencil=false
     fontMaterial=null
     fontObject=&Roboto
     letterSpacing=0
     monospacedWidth=1
     size=12
     skewAmount=0
     typefaceFontName=Bold
     outlineSettings {
      bApplyOutlineToDropShadows=false
      bMiteredCorners=false
      bSeparateFillAlpha=false
      outlineMaterial=null
      outlineSize=0
      outlineColor=<rgba 0 0 0 1>
     }
    }
   }
   slot {class=SizeBoxSlot}
  }
  Spacer_96 {widgetClassPath=&Spacer, parent=HorizontalBox, widget=Spacer_96, properties {size=<2 2>}, slot {class=HorizontalBoxSlot}}
  ValueTextBlock {
   widgetClassPath=&TextBlock
   bInherited=true
   parent=HorizontalBox
   widget=ValueTextBlock
   properties {
    colorAndOpacity {colorUseRule=UseColor_Specified, specifiedColor=<rgba 0.5765 0.5765 0.5765 1>}
    font {
     bForceMonospaced=false
     bMaterialIsStencil=false
     fontMaterial=null
     fontObject=&Roboto
     letterSpacing=0
     monospacedWidth=1
     size=12
     skewAmount=0
     typefaceFontName=Bold
     outlineSettings {
      bApplyOutlineToDropShadows=false
      bMiteredCorners=false
      bSeparateFillAlpha=false
      outlineMaterial=null
      outlineSize=0
      outlineColor=<rgba 0 0 0 1>
     }
    }
   }
   slot {class=HorizontalBoxSlot}
  }
 ]
}
